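; NPC music component for Lesser Demons that drives combat music phases from the owning boss pawn's health.
; On health change, it sets a music float parameter to phase 1 or 2 against a health threshold, rechecking when a player enters the AGV.
; With debug enabled in editor or PIE, it watches the MetaSound CurrentPhase output and shows it on screen and in color-coded text above the NPC.
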(event Custom|OnOutputChanged
  (params (OutputName :type Name) (Output :type Meta%20Sound%20Output%20Structure%20%28by%20ref%29))
  ;[0.15 0 0.1026 0.5] Debug Print above Character Head
  (bind (_success _returnvalue) (MetaSoundOutput|GetInt32 Output))
  (Development|PrintString (Utilities|String|Append "CurrentPhase: " (Utilities|String|ToString(Integer) _returnvalue)) true true "(R=0.000000,G=1.000000,B=0.843891,A=1.000000)" 1e+01 "CurrentPhase")
  (bind _debugtextrendercomponent (Variables|Default|GetDebugTextRenderComponent)
    (:then
      (Rendering|Components|TextRender|SetText _debugtextrendercomponent (Utilities|Text|MakeLiteralText (Utilities|Text|ToText(String) (Utilities|String|Append "CurrentPhase: " (Utilities|String|ToString(Integer) _returnvalue)))))
      (Rendering|Components|TextRender|SetTextRenderColor _debugtextrendercomponent (Math|Conversions|ToColor(LinearColor) (Math|Color|MakeColor (Math|Float|MapRangeClamped (Math|Conversions|ToFloat(Integer) _returnvalue) 0.0 2.0 0.0 1.0) (Math|Float|MapRangeClamped (Math|Conversions|ToFloat(Integer) _returnvalue) 0.0 2.0 1.0)))))
    (:else)))

(event Custom|DebugMusicComponent
  (params (Debug :type Boolean))
  (Variables|Default|SetDebug Debug))

(fn OnHealthChange
  (|CallParentFunction)
  (bind _asbp_player_pawn_non_participant_boss (Utilities|Casting|CastToBP_PlayerPawn_NonParticipant_Boss (Variables|Private|GetNPCOwner)))
  (bind _healthinfo_healthpercent_2_7edada384c7908ea98debd9958a249de (Utilities|Struct|BreakStructNPCHealthInfo (Variables|/Game/Athena/AI/NPCs/Base/BP_PlayerPawn_NonParticipant_Boss.BP_PlayerPawn_NonParticipant_Boss_C|GetHealthInfo_HealthPercent_2_7EDADA384C7908EA98DEBD9958A249DE _asbp_player_pawn_non_participant_boss)))
  (if (>= _healthinfo_healthpercent_2_7edada384c7908ea98debd9958a249de (Variables|Default|GetPhase2HealthThreshold))
    (Class|NPCMusicManagerComponent|SetMusicFloatParam (Variables|Private|GetMusicManager) self (Variables|MetaSoundParams|GetOnEnterCombatPhaseName) 1.0)
    (else
      (Class|NPCMusicManagerComponent|SetMusicFloatParam (Variables|Private|GetMusicManager) self (Variables|MetaSoundParams|GetOnEnterCombatPhaseName) 2.0))))

(fn OnPlayerEnterAGV
  (bind (_isvalid _groupcomponent) (Class|NPCMusicManagerComponent|GetGroupComponent (Variables|Private|GetMusicManager) self))
  (|CallParentFunction)
  (Private|Delegates|OnHealthChange)
  (if (and (Editor|InEditororPIE) (Variables|Default|GetDebug))
    (MetaSoundOutput|WatchOutput (WorldSubsystems|GetMetaSoundOutputSubsystem) (Variables|/FortNPCMusic/Audio/NPCMusicActorGroupComponent.NPCMusicActorGroupComponent_C|GetMusic _groupcomponent) "CurrentPhase" (EventDispatchers|CreateEvent :!event OnOutputChanged) "None" "None" :!class /Script/MetasoundEngine.MetaSoundOutputSubsystem)))

(fn SetupPawnBindings
  (|CallParentFunction)
  ;[0 0.1134 0.15 0.5] Sets Up Text Render Component
  (if (and (Editor|InEditororPIE) (Variables|Default|GetDebug))
    (bind _returnvalue (Game|AddComponentbyClass (Variables|Private|GetNPCOwner) "/Script/Engine.TextRenderComponent" false (struct :Location "0.000000,0.000000,100.000000" :Rotation "0, 0, 0" :Scale "1.000000,1.000000,1.000000")))
    (bind _output_get (Variables|Default|SetDebugTextRenderComponent _returnvalue))
    (Rendering|Components|TextRender|SetFont _output_get "/Game/UI/Fonts/WorldFonts/Font_BurbankRegBlack_DistanceField.Font_BurbankRegBlack_DistanceField")
    (Rendering|Components|TextRender|SetXScale _output_get 2.5)
    (Rendering|Components|TextRender|SetYScale _output_get 2.5)))
