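# Sound cue for the Lesser Demon derez sound effect, routed through the Fort_BossFight sound class.
# A single wave player feeds the LesserDemon_Derez wave straight to the output with a boosted volume multiplier and no randomization.
# A dedicated attenuation asset shapes its spatial falloff.
# Overridden concurrency allows only one instance at a time, with a retrigger time to limit rapid repeats.
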SoundCue /LesserDemons/Audio/SC_LesserDemon_Derez
root=WavePlayer_0
nodes [{id=WavePlayer_0, properties {soundWaveAssetPtr=/LesserDemons/Audio/SFX/LesserDemon_Derez}}]
properties {
 attenuationSettings=&/SA_LesserDemon_Derez
 bOverrideConcurrency=true
 maxDistance=4000
 soundClassObject=&Fort_BossFight
 volumeMultiplier=3
 concurrencyOverrides {maxCount=1, retriggerTime=2}
}
